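# Level Sequence for the QuarterClaspJoin_F custom skin's main-player menu animation in Battle Royale, apparently an animation authoring scene.
# A single spawned FortPlayerMannequin holds a constant, slightly offset transform and plays a skeletal animation section over frames 0–300 at 30 fps.
# A Fortnite_Legacy_Face_CtrlRig track is layered on the same binding, with only constant master-control placement, head-follow and flipbook-off channels, so it carries no facial motion.
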LevelSequence /BRCosmetics/Animation/Game/MainPlayer/Menu/BR/CustomSkins/_s40/QuarterClaspJoin_F/QCJ_Anim/QCJ_LS
bindings [FortPlayerMannequin {
  id=#1
  possessed_class=FortPlayerMannequin
  tracks [
   {type=MovieSceneSpawnTrack, sections [{type=MovieSceneSpawnSection, channels [None {type=MovieSceneScriptingBoolChannel, default=true}]}]}
   {
    type=MovieScene3DTransformTrack
    sections [{
      type=MovieScene3DTransformSection
      channels [
       Location.X {type=MovieSceneScriptingDoubleChannel, default=1.1853, const [1.1853,10]}
       Location.Y {type=MovieSceneScriptingDoubleChannel, default=0.0368, const [0.0368,10]}
       Location.Z {type=MovieSceneScriptingDoubleChannel, default=-0.0406, const [-0.0406,10]}
       Rotation.X {type=MovieSceneScriptingDoubleChannel, default=0, const [0,10]}
       Rotation.Y {type=MovieSceneScriptingDoubleChannel, default=0, const [0,10]}
       Rotation.Z {type=MovieSceneScriptingDoubleChannel, default=0, const [0,10]}
       Scale.X {type=MovieSceneScriptingDoubleChannel, default=1, const [1,10]}
       Scale.Y {type=MovieSceneScriptingDoubleChannel, default=1, const [1,10]}
       Scale.Z {type=MovieSceneScriptingDoubleChannel, default=1, const [1,10]}
      ]
     }]
   }
   {type=MovieSceneSkeletalAnimationTrack, sections [{type=MovieSceneSkeletalAnimationSection, end_frame=300, start_frame=0}]}
   {
    type=MovieSceneControlRigParameterTrack
    display_name=Fortnite_Legacy_Face_CtrlRig
    sections [{
      type=MovieSceneControlRigParameterSection
      channels [
       Flipbook {type=MovieSceneScriptingBoolChannel, default=false, const [false,3]}
       Face_Ctrl_Master.Location.X {type=MovieSceneScriptingFloatChannel, default=-2.4905, const [-2.4905,3]}
       Face_Ctrl_Master.Location.Y {type=MovieSceneScriptingFloatChannel, default=-5.0193, const [-5.0193,3]}
       Face_Ctrl_Master.Location.Z {type=MovieSceneScriptingFloatChannel, default=-7.7897, const [-7.7897,3]}
       Face_Ctrl_Master.Rotation.X {type=MovieSceneScriptingFloatChannel, default=-0.0286, const [-0.0286,3]}
       Face_Ctrl_Master.Rotation.Y {type=MovieSceneScriptingFloatChannel, default=-0.162, const [-0.162,3]}
       Face_Ctrl_Master.Rotation.Z {type=MovieSceneScriptingFloatChannel, default=19.9994, const [19.9994,3]}
       Face_Ctrl_Master.Scale.X {type=MovieSceneScriptingFloatChannel, default=1, const [1,3]}
       Face_Ctrl_Master.Scale.Y {type=MovieSceneScriptingFloatChannel, default=1, const [1,3]}
       Face_Ctrl_Master.Scale.Z {type=MovieSceneScriptingFloatChannel, default=1, const [1,3]}
       gui_follow_head {type=MovieSceneScriptingFloatChannel, default=1, const [1,3]}
      ]
     }]
   }
  ]
 }]
display_rate {denominator=1, numerator=30}
tick_resolution {denominator=1, numerator=24000}
tracks []
